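# Niagara system for the ProngCling pickaxe cosmetic, a melee weapon effect.
# It contains three emitters named as sword smoke variants: SwordSmoke, SwordSmoke001 and SwordSmoke003.
# The system uses only the SystemState module, and no renderers, data interfaces, dynamic inputs or user variables are recorded for it.
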NiagaraSystem /BRCosmetics/Weapons/FORT_Melee/Pickaxe_ProngCling/FX/NS_Pickaxe_ProngCling
dependencies {usedDataInterfaces [], usedDynamicInputs [], usedModules [&SystemState]}
emitters {
 SwordSmoke {rendererClasses [], renderers []}
 SwordSmoke001 {rendererClasses [], renderers []}
 SwordSmoke003 {rendererClasses [], renderers []}
}
summary {
 systemName=NS_Pickaxe_ProngCling
 emitters [
  {emitterName=SwordSmoke, rendererClasses []}
  {emitterName=SwordSmoke001, rendererClasses []}
  {emitterName=SwordSmoke003, rendererClasses []}
 ]
 userVariables []
}
